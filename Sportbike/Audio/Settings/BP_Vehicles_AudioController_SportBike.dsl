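; Vehicle audio controller Blueprint for the sport bike, handling super-slide audio and the boost motor simulation.
; On init, it casts the vehicle actor to the sport motorcycle and binds to its super-sliding changed delegate.
; Sliding spawns an attached slide sound and stores its component, which fades out linearly when sliding stops.
; Its AddMotorComponents override registers the boost motor sim before calling the parent's setup.
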(event VehicleAudio|EventOnInit
  (|CallParentFunction)
  (bind _asmotorcycle_sport_vehicle (Utilities|Casting|CastToMotorcycle_Sport_Vehicle (AudioController|GetVehicleActor)))
  (Motorcycle|BindEventtoOnSuperSlidingChangedDelegate _asmotorcycle_sport_vehicle (EventDispatchers|CreateEvent :!event Custom|SuperSlideEvent)))

(event Custom|SuperSlideEvent
  (params (bIsSuperSliding :type Boolean))
  (if bIsSuperSliding
    (bind _slidesound (Variables|Data|GetSlideSound)
      (:then
        (bind _returnvalue (Audio|SpawnSoundAttached _slidesound (Variables|Transformation|GetRootComponent)))
        (Variables|Default|SetSlideAudio _returnvalue))
      (:else))
    (else
      (bind _slideaudio (Variables|Default|GetSlideAudio)
        (:then
          (Audio|Components|Audio|FadeOut _slideaudio 0.55 0.0 "Linear" :!class /Script/Engine.AudioComponent))
        (:else)))))

(fn AddMotorComponents
  (MotorModel|AddMotorSimComponentOrder (Variables|Default|GetBoostMotorSim))
  (|CallParentFunction))

(fn ConstructionScript
  (|CallParentFunction))
